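# Physics asset for the NightBloom male medium cosmetic body, providing secondary motion for its coat and skirt bones.
# Locked capsules on pelvis, spine_01, legs and feet form the body.
# Coat chains hang from zero-size anchors on spine_01, with swing limits widening toward the hem; skirt chains hang from the thighs with free rotation.
# Extra locked cross-constraints tie side coat panels to front and back neighbours, and join the two bA tips.
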PhysicsAsset /BRCosmetics/Characters/Player/Male/Medium/Bodies/M_MED_NightBloom/Meshes/M_MED_NightBloom_Physics1
bodies {
 base_Coat_bA_l_01 [{shapeType=Capsule, length=0, radius=0, shapeName=base_Coat_bA_l_01_capsule}]
 base_Coat_bA_r_01 [{shapeType=Capsule, length=0, radius=0, shapeName=base_Coat_bA_r_01_capsule}]
 base_Coat_b_l_01 [{shapeType=Capsule, length=0, radius=0, shapeName=base_Coat_b_l_01_capsule}]
 base_Coat_b_r_01 [{shapeType=Capsule, length=0, radius=0, shapeName=base_Coat_b_r_01_capsule}]
 base_Coat_f_l_01 [{shapeType=Capsule, length=0, radius=0, shapeName=base_Coat_f_l_01_capsule}]
 base_Coat_f_r_01 [{shapeType=Capsule, length=0, radius=0, shapeName=base_Coat_f_r_01_capsule}]
 base_Coat_l_01 [{shapeType=Capsule, length=0, radius=0, shapeName=base_Coat_l_01_capsule}]
 base_Coat_r_01 [{shapeType=Capsule, length=0, radius=0, shapeName=base_Coat_r_01_capsule}]
 base_Skirt_b_l_01 [{shapeType=Capsule, length=0, radius=0, shapeName=base_Skirt_b_l_01_capsule}]
 base_Skirt_b_r_01 [{shapeType=Capsule, length=0, radius=0, shapeName=base_Skirt_b_r_01_capsule}]
 calf_l [{shapeType=Capsule, length=29.7, radius=10, shapeName=calf_l_capsule, center=<-15.399 -0.3524 0.8897>, rotation=<rot 90 180 0>}]
 calf_r [{shapeType=Capsule, length=29.7, radius=10, shapeName=calf_r_capsule, center=<15.399 0.3524 -0.8897>, rotation=<rot 90 0 0>}]
 dyn_Coat_bA_l_01 [{shapeType=Capsule, length=6.8, radius=4.5, shapeName=dyn_Coat_bA_l_01_capsule, center=<6.2696 0 2.2486>, rotation=<rot 0 0 90>}]
 dyn_Coat_bA_l_02 [{shapeType=Capsule, length=6.8, radius=4.5, shapeName=dyn_Coat_bA_l_02_capsule, center=<6.2696 0 1.7652>, rotation=<rot 0 0 90>}]
 dyn_Coat_bA_l_03 [{
   shapeType=Capsule
   length=6.8
   radius=4.5
   shapeName=dyn_Coat_bA_l_03_capsule
   center=<6.426 -0.0167 2.0866>
   rotation=<rot -0.0234 -6.1179 90.0009>
  }]
 dyn_Coat_bA_r_01 [{
   shapeType=Capsule
   length=6.8
   radius=4.5
   shapeName=dyn_Coat_bA_r_01_capsule
   center=<6.2697 0.2773 2.2314>
   rotation=<rot -0.0001 0 97.084>
  }]
 dyn_Coat_bA_r_02 [{
   shapeType=Capsule
   length=6.8
   radius=4.5
   shapeName=dyn_Coat_bA_r_02_capsule
   center=<6.2697 0.2177 1.7517>
   rotation=<rot -0.0001 0 97.084>
  }]
 dyn_Coat_bA_r_03 [{
   shapeType=Capsule
   length=6.8
   radius=4.5
   shapeName=dyn_Coat_bA_r_03_capsule
   center=<6.4995 0.2759 2.0684>
   rotation=<rot -0.5724 4.6101 97.0609>
  }]
 dyn_Coat_b_l_01 [{shapeType=Capsule, length=7.3, radius=3.8, shapeName=dyn_Coat_b_l_01_capsule, center=<5.4746 0 2.7589>, rotation=<rot 0 0 90>}]
 dyn_Coat_b_l_02 [{shapeType=Capsule, length=7, radius=4.1, shapeName=dyn_Coat_b_l_02_capsule, center=<5.4746 0 2.8026>, rotation=<rot 0 0 90>}]
 dyn_Coat_b_l_03 [{
   shapeType=Capsule
   length=7
   radius=4.1
   shapeName=dyn_Coat_b_l_03_capsule
   center=<7.7702 -0.4554 4.0763>
   rotation=<rot -3.0644 -12.8538 77.0505>
  }]
 dyn_Coat_b_r_01 [{shapeType=Capsule, length=7.3, radius=3.8, shapeName=dyn_Coat_b_r_01_capsule, center=<5.4746 0 2.8664>, rotation=<rot 0 0 90>}]
 dyn_Coat_b_r_02 [{shapeType=Capsule, length=7, radius=4.1, shapeName=dyn_Coat_b_r_02_capsule, center=<5.4746 0 3.0869>, rotation=<rot 0 0 90>}]
 dyn_Coat_b_r_03 [{
   shapeType=Capsule
   length=7
   radius=4.1
   shapeName=dyn_Coat_b_r_03_capsule
   center=<7.9569 0.4771 2.4959>
   rotation=<rot -0.0001 13.6442 99.0272>
  }]
 dyn_Coat_f_l_01 [{
   shapeType=Capsule
   length=4
   radius=5.3
   shapeName=dyn_Coat_f_l_01_capsule
   center=<5.4972 -6.0302 -0.0028>
   rotation=<rot 0.0623 -0.0486 51.8865>
  }]
 dyn_Coat_f_l_02 [{
   shapeType=Capsule
   length=8.9
   radius=5.4
   shapeName=dyn_Coat_f_l_02_capsule
   center=<5.4745 -6.2603 -0.4357>
   rotation=<rot 0.0235 0 51.1614>
  }]
 dyn_Coat_f_l_03 [{
   shapeType=Capsule
   length=9.5
   radius=5.2
   shapeName=dyn_Coat_f_l_03_capsule
   center=<4.3851 -7.0473 -0.4505>
   rotation=<rot -1.9752 -2.2443 57.5309>
  }]
 dyn_Coat_f_l_04 [{
   shapeType=Capsule
   length=8.3
   radius=6.5
   shapeName=dyn_Coat_f_l_04_capsule
   center=<4.8005 -6.6402 2.1057>
   rotation=<rot -0.015 -4.2616 65.9155>
  }]
 dyn_Coat_f_r_01 [{
   shapeType=Capsule
   length=2.8
   radius=4.4
   shapeName=dyn_Coat_f_r_01_capsule
   center=<6.7587 3.5677 1.0704>
   rotation=<rot -6.7928 -4.4154 116.3401>
  }]
 dyn_Coat_f_r_02 [{
   shapeType=Capsule
   length=5.4
   radius=4.5
   shapeName=dyn_Coat_f_r_02_capsule
   center=<5.1669 4.1902 -0.0404>
   rotation=<rot 0.0007 0 119.4974>
  }]
 dyn_Coat_f_r_03 [{
   shapeType=Capsule
   length=8.4
   radius=5.5
   shapeName=dyn_Coat_f_r_03_capsule
   center=<5.5796 7.2986 -1.4107>
   rotation=<rot -5.5106 -2.7824 132.6384>
  }]
 dyn_Coat_f_r_04 [{
   shapeType=Capsule
   length=8.7
   radius=6.1
   shapeName=dyn_Coat_f_r_04_capsule
   center=<6.6306 7.8849 -0.6602>
   rotation=<rot -2.4418 2.2227 127.7684>
  }]
 dyn_Coat_l_01 [{shapeType=Capsule, length=9, radius=4.8, shapeName=dyn_Coat_l_01_capsule, center=<5.4746 0 2.9271>, rotation=<rot 0 0 90>}]
 dyn_Coat_l_02 [{shapeType=Capsule, length=9.6, radius=4.2, shapeName=dyn_Coat_l_02_capsule, center=<5.4746 0 3.3393>, rotation=<rot 0 0 90>}]
 dyn_Coat_l_03 [{
   shapeType=Capsule
   length=8.3
   radius=5.5
   shapeName=dyn_Coat_l_03_capsule
   center=<7.3246 0.1254 2.791>
   rotation=<rot 0.5272 -14.1347 101.8562>
  }]
 dyn_Coat_r_01 [{shapeType=Capsule, length=9, radius=4.8, shapeName=dyn_Coat_r_01_capsule, center=<5.4746 0 2.936>, rotation=<rot 0 -0.0001 79.3369>}]
 dyn_Coat_r_02 [{shapeType=Capsule, length=9.6, radius=4.2, shapeName=dyn_Coat_r_02_capsule, center=<5.4746 0 3.3581>, rotation=<rot 0 0 90>}]
 dyn_Coat_r_03 [{shapeType=Capsule, length=8.1, radius=5.7, shapeName=dyn_Coat_r_03_capsule, center=<6.8225 0.1806 4.3415>, rotation=<rot 0 13.9805 90.0001>}]
 dyn_Skirt_b_l_01 [{shapeType=Capsule, length=6.1, radius=5.5, shapeName=dyn_Skirt_b_l_01_capsule, center=<5.1545 0 -3.5801>, rotation=<rot 0 0 90>}]
 dyn_Skirt_b_l_02 [{shapeType=Capsule, length=6.1, radius=5.5, shapeName=dyn_Skirt_b_l_02_capsule, center=<5.1545 0 -3.5801>, rotation=<rot 0 0 90>}]
 dyn_Skirt_b_l_03 [{shapeType=Capsule, length=6.1, radius=5.5, shapeName=dyn_Skirt_b_l_03_capsule, center=<5.1545 0 -3.5801>, rotation=<rot 0 0 90>}]
 dyn_Skirt_b_r_01 [{shapeType=Capsule, length=6.1, radius=5.5, shapeName=dyn_Skirt_b_r_01_capsule, center=<5.1545 0 -3.5801>, rotation=<rot 0 0 90>}]
 dyn_Skirt_b_r_02 [{shapeType=Capsule, length=6.1, radius=5.5, shapeName=dyn_Skirt_b_r_02_capsule, center=<5.1545 0 -3.5801>, rotation=<rot 0 0 90>}]
 dyn_Skirt_b_r_03 [{shapeType=Capsule, length=6.1, radius=5.5, shapeName=dyn_Skirt_b_r_03_capsule, center=<5.1545 0 -3.5801>, rotation=<rot 0 0 90>}]
 dyn_Skirt_f_l_02 [{
   shapeType=Capsule
   length=7.6
   radius=7.6918
   shapeName=dyn_Skirt_f_l_02_capsule
   center=<1.7034 -2.6879 -7.1688>
   rotation=<rot 32.3189 87.7798 88.8133>
  }]
 dyn_Skirt_f_l_03 [{
   shapeType=Capsule
   length=7.5
   radius=8.9225
   shapeName=dyn_Skirt_f_l_03_capsule
   center=<9.9465 -2.3961 -8.275>
   rotation=<rot 22.9424 88.1166 88.4069>
  }]
 dyn_Skirt_f_r_02 [{
   shapeType=Capsule
   length=13.4
   radius=6.1517
   shapeName=dyn_Skirt_f_r_02_capsule
   center=<5.0913 1.9259 -5.3037>
   rotation=<rot -23.4102 86.9464 91.2144>
  }]
 dyn_Skirt_f_r_03 [{
   shapeType=Capsule
   length=12.7
   radius=7.6129
   shapeName=dyn_Skirt_f_r_03_capsule
   center=<8.3664 1.9259 -5.3037>
   rotation=<rot -23.4102 86.9464 91.2144>
  }]
 dyn_Skirt_l_02 [{
   shapeType=Capsule
   length=6.8
   radius=6
   shapeName=dyn_Skirt_l_02_capsule
   center=<3.6165 -1.9229 -3.6294>
   rotation=<rot -7.0868 88.1094 90.2334>
  }]
 dyn_Skirt_l_03 [{
   shapeType=Capsule
   length=6.8
   radius=6
   shapeName=dyn_Skirt_l_03_capsule
   center=<8.9234 -2.0954 -3.6079>
   rotation=<rot -7.0868 88.1094 90.2334>
  }]
 dyn_Skirt_r_02 [{shapeType=Capsule, length=13, radius=5.4, shapeName=dyn_Skirt_r_02_capsule, center=<4.4287 0 -4.1246>, rotation=<rot -0.0023 88.7634 90>}]
 dyn_Skirt_r_03 [{shapeType=Capsule, length=13, radius=5.4, shapeName=dyn_Skirt_r_03_capsule, center=<4.4287 0 -4.1246>, rotation=<rot -0.0023 88.7634 90>}]
 foot_l [{
   shapeType=Capsule
   length=14.5035
   radius=7.9009
   shapeName=foot_l_capsule
   center=<-3.0533 6.3556 -0.0602>
   rotation=<rot 0 6.9394 -88.7683>
  }]
 foot_r [{shapeType=Capsule, length=14.502, radius=7.9036, shapeName=foot_r_capsule, center=<3.057 -6.3574 0.0612>, rotation=<rot 0 6.9736 -88.7774>}]
 pelvis [{shapeType=Capsule, length=9, radius=10.7, shapeName=pelvis_capsule}]
 spine_01 [{shapeType=Capsule, length=4.9, radius=11.2, shapeName=spine_01_capsule, center=<1.7641 0.2134 0>}]
 thigh_l [
  {
   shapeType=Capsule
   length=26.2
   radius=10.8
   shapeName=thigh_l_capsule
   center=<-18.4763 -2.1078 0.3234>
   rotation=<rot 87.9682 179.0098 -0.0006>
  }
  {shapeType=Capsule, length=12.3, radius=10.2, shapeName=thigh_r_capsule, center=<-18.119 -3.4174 -0.8897>, rotation=<rot 90 9.9353 0>}
 ]
 thigh_r [
  {
   shapeType=Capsule
   length=12.3
   radius=10.2
   shapeName=thigh_r_capsule
   center=<18.0655 2.9997 -0.8956>
   rotation=<rot 88.084 -177.5649 173.5121>
  }
  {
   shapeType=Capsule
   length=25.2
   radius=11.8
   shapeName=thigh_r_capsule
   center=<18.5168 1.6558 -0.8897>
   rotation=<rot 86.8435 177.7696 179.9993>
  }
 ]
}
constraints [
 {bone1Name=dyn_Coat_bA_l_01, bone2Name=base_Coat_bA_l_01, swing1LimitDegrees=5, swing2LimitDegrees=10, twistLimitDegrees=2}
 {bone1Name=dyn_Coat_bA_l_02, bone2Name=dyn_Coat_bA_l_01, swing1LimitDegrees=10, swing2LimitDegrees=10, twistLimitDegrees=2}
 {bone1Name=dyn_Coat_bA_l_03, bone2Name=dyn_Coat_bA_l_02, swing1LimitDegrees=15, swing2LimitDegrees=15, twistLimitDegrees=2}
 {bone1Name=dyn_Coat_bA_r_01, bone2Name=base_Coat_bA_r_01, swing1LimitDegrees=5, swing2LimitDegrees=10, twistLimitDegrees=2}
 {bone1Name=dyn_Coat_bA_r_02, bone2Name=dyn_Coat_bA_r_01, swing1LimitDegrees=10, swing2LimitDegrees=10, twistLimitDegrees=2}
 {bone1Name=dyn_Coat_bA_r_03, bone2Name=dyn_Coat_bA_r_02, swing1LimitDegrees=15, swing2LimitDegrees=15, twistLimitDegrees=2}
 {bone1Name=dyn_Coat_b_l_01, bone2Name=base_Coat_b_l_01, swing1LimitDegrees=5, swing2LimitDegrees=10, twistLimitDegrees=2}
 {bone1Name=dyn_Coat_b_l_02, bone2Name=dyn_Coat_b_l_01, swing1LimitDegrees=10, swing2LimitDegrees=10, twistLimitDegrees=2}
 {bone1Name=dyn_Coat_b_l_03, bone2Name=dyn_Coat_b_l_02, swing1LimitDegrees=15, swing2LimitDegrees=15, twistLimitDegrees=2}
 {bone1Name=dyn_Coat_b_r_01, bone2Name=base_Coat_b_r_01, swing1LimitDegrees=5, swing2LimitDegrees=10, twistLimitDegrees=2}
 {bone1Name=dyn_Coat_b_r_02, bone2Name=dyn_Coat_b_r_01, swing1LimitDegrees=10, swing2LimitDegrees=10, twistLimitDegrees=2}
 {bone1Name=dyn_Coat_b_r_03, bone2Name=dyn_Coat_b_r_02, swing1LimitDegrees=15, swing2LimitDegrees=15, twistLimitDegrees=2}
 {bone1Name=dyn_Coat_f_l_01, bone2Name=base_Coat_f_l_01, swing1LimitDegrees=5, swing2LimitDegrees=5, twistLimitDegrees=2}
 {bone1Name=dyn_Coat_f_l_02, bone2Name=dyn_Coat_f_l_01, swing1LimitDegrees=5, swing2LimitDegrees=10, twistLimitDegrees=2}
 {bone1Name=dyn_Coat_f_l_03, bone2Name=dyn_Coat_f_l_02, swing1LimitDegrees=10, swing2LimitDegrees=10, twistLimitDegrees=2}
 {bone1Name=dyn_Coat_f_l_04, bone2Name=dyn_Coat_f_l_03, swing1LimitDegrees=15, swing2LimitDegrees=20, twistLimitDegrees=2}
 {bone1Name=dyn_Coat_f_r_01, bone2Name=base_Coat_f_r_01, swing1LimitDegrees=5, swing2LimitDegrees=5, twistLimitDegrees=2}
 {bone1Name=dyn_Coat_f_r_02, bone2Name=dyn_Coat_f_r_01, swing1LimitDegrees=5, swing2LimitDegrees=10, twistLimitDegrees=2}
 {bone1Name=dyn_Coat_f_r_03, bone2Name=dyn_Coat_f_r_02, swing1LimitDegrees=5, swing2LimitDegrees=10, twistLimitDegrees=2}
 {bone1Name=dyn_Coat_f_r_04, bone2Name=dyn_Coat_f_r_03, swing1LimitDegrees=15, swing2LimitDegrees=20, twistLimitDegrees=2}
 {bone1Name=dyn_Coat_l_01, bone2Name=base_Coat_l_01, swing1LimitDegrees=5, swing2LimitDegrees=10, twistLimitDegrees=2}
 {bone1Name=dyn_Coat_l_02, bone2Name=dyn_Coat_l_01, swing1LimitDegrees=10, swing2LimitDegrees=10, twistLimitDegrees=2}
 {bone1Name=dyn_Coat_l_03, bone2Name=dyn_Coat_l_02, swing1LimitDegrees=15, swing2LimitDegrees=15, twistLimitDegrees=2}
 {bone1Name=dyn_Coat_r_01, bone2Name=base_Coat_r_01, swing1LimitDegrees=5, swing2LimitDegrees=10, twistLimitDegrees=2}
 {bone1Name=dyn_Coat_r_02, bone2Name=dyn_Coat_r_01, swing1LimitDegrees=5, swing2LimitDegrees=10, twistLimitDegrees=2}
 {bone1Name=dyn_Coat_r_03, bone2Name=dyn_Coat_r_02, swing1LimitDegrees=15, swing2LimitDegrees=15, twistLimitDegrees=2}
 {
  bone1Name=base_Coat_l_01
  bone2Name=spine_01
  swing1LimitDegrees=45
  swing1Motion=Locked
  swing2LimitDegrees=45
  swing2Motion=Locked
  twistLimitDegrees=45
  twistMotion=Locked
 }
 {
  bone1Name=base_Coat_b_l_01
  bone2Name=spine_01
  swing1LimitDegrees=45
  swing1Motion=Locked
  swing2LimitDegrees=45
  swing2Motion=Locked
  twistLimitDegrees=45
  twistMotion=Locked
 }
 {
  bone1Name=base_Coat_bA_l_01
  bone2Name=spine_01
  swing1LimitDegrees=45
  swing1Motion=Locked
  swing2LimitDegrees=45
  swing2Motion=Locked
  twistLimitDegrees=45
  twistMotion=Locked
 }
 {
  bone1Name=base_Coat_bA_r_01
  bone2Name=spine_01
  swing1LimitDegrees=45
  swing1Motion=Locked
  swing2LimitDegrees=45
  swing2Motion=Locked
  twistLimitDegrees=45
  twistMotion=Locked
 }
 {
  bone1Name=base_Coat_b_r_01
  bone2Name=spine_01
  swing1LimitDegrees=45
  swing1Motion=Locked
  swing2LimitDegrees=45
  swing2Motion=Locked
  twistLimitDegrees=45
  twistMotion=Locked
 }
 {
  bone1Name=base_Coat_r_01
  bone2Name=spine_01
  swing1LimitDegrees=45
  swing1Motion=Locked
  swing2LimitDegrees=45
  swing2Motion=Locked
  twistLimitDegrees=45
  twistMotion=Locked
 }
 {
  bone1Name=base_Coat_f_r_01
  bone2Name=spine_01
  swing1LimitDegrees=45
  swing1Motion=Locked
  swing2LimitDegrees=45
  swing2Motion=Locked
  twistLimitDegrees=45
  twistMotion=Locked
 }
 {
  bone1Name=base_Coat_f_l_01
  bone2Name=spine_01
  swing1LimitDegrees=45
  swing1Motion=Locked
  swing2LimitDegrees=45
  swing2Motion=Locked
  twistLimitDegrees=45
  twistMotion=Locked
 }
 {
  bone1Name=spine_01
  bone2Name=pelvis
  swing1LimitDegrees=45
  swing1Motion=Locked
  swing2LimitDegrees=45
  swing2Motion=Locked
  twistLimitDegrees=45
  twistMotion=Locked
 }
 {
  bone1Name=thigh_r
  bone2Name=pelvis
  swing1LimitDegrees=45
  swing1Motion=Locked
  swing2LimitDegrees=45
  swing2Motion=Locked
  twistLimitDegrees=45
  twistMotion=Locked
 }
 {
  bone1Name=calf_r
  bone2Name=thigh_r
  swing1LimitDegrees=45
  swing1Motion=Locked
  swing2LimitDegrees=45
  swing2Motion=Locked
  twistLimitDegrees=45
  twistMotion=Locked
 }
 {
  bone1Name=thigh_l
  bone2Name=pelvis
  swing1LimitDegrees=45
  swing1Motion=Locked
  swing2LimitDegrees=45
  swing2Motion=Locked
  twistLimitDegrees=45
  twistMotion=Locked
 }
 {
  bone1Name=calf_l
  bone2Name=thigh_l
  swing1LimitDegrees=45
  swing1Motion=Locked
  swing2LimitDegrees=45
  swing2Motion=Locked
  twistLimitDegrees=45
  twistMotion=Locked
 }
 {
  bone1Name=dyn_Skirt_f_r_02
  bone2Name=thigh_r
  swing1LimitDegrees=45
  swing1Motion=Free
  swing2LimitDegrees=45
  swing2Motion=Free
  twistLimitDegrees=45
  twistMotion=Free
 }
 {
  bone1Name=dyn_Skirt_f_r_03
  bone2Name=dyn_Skirt_f_r_02
  swing1LimitDegrees=45
  swing1Motion=Free
  swing2LimitDegrees=45
  swing2Motion=Free
  twistLimitDegrees=45
  twistMotion=Free
 }
 {
  bone1Name=dyn_Skirt_f_l_02
  bone2Name=thigh_l
  swing1LimitDegrees=45
  swing1Motion=Free
  swing2LimitDegrees=45
  swing2Motion=Free
  twistLimitDegrees=45
  twistMotion=Free
 }
 {
  bone1Name=dyn_Skirt_f_l_03
  bone2Name=dyn_Skirt_f_l_02
  swing1LimitDegrees=45
  swing1Motion=Free
  swing2LimitDegrees=45
  swing2Motion=Free
  twistLimitDegrees=45
  twistMotion=Free
 }
 {
  bone1Name=dyn_Skirt_r_02
  bone2Name=thigh_r
  swing1LimitDegrees=45
  swing1Motion=Free
  swing2LimitDegrees=45
  swing2Motion=Free
  twistLimitDegrees=45
  twistMotion=Free
 }
 {
  bone1Name=dyn_Skirt_r_03
  bone2Name=dyn_Skirt_r_02
  swing1LimitDegrees=45
  swing1Motion=Free
  swing2LimitDegrees=45
  swing2Motion=Free
  twistLimitDegrees=45
  twistMotion=Free
 }
 {
  bone1Name=dyn_Skirt_l_02
  bone2Name=thigh_l
  swing1LimitDegrees=45
  swing1Motion=Free
  swing2LimitDegrees=45
  swing2Motion=Free
  twistLimitDegrees=45
  twistMotion=Free
 }
 {
  bone1Name=dyn_Skirt_l_03
  bone2Name=dyn_Skirt_l_02
  swing1LimitDegrees=45
  swing1Motion=Free
  swing2LimitDegrees=45
  swing2Motion=Free
  twistLimitDegrees=45
  twistMotion=Free
 }
 {
  bone1Name=foot_l
  bone2Name=calf_l
  swing1LimitDegrees=45
  swing1Motion=Locked
  swing2LimitDegrees=45
  swing2Motion=Locked
  twistLimitDegrees=45
  twistMotion=Locked
 }
 {
  bone1Name=foot_r
  bone2Name=calf_r
  swing1LimitDegrees=45
  swing1Motion=Locked
  swing2LimitDegrees=45
  swing2Motion=Locked
  twistLimitDegrees=45
  twistMotion=Locked
 }
 {
  bone1Name=base_Skirt_b_l_01
  bone2Name=thigh_l
  swing1LimitDegrees=45
  swing1Motion=Locked
  swing2LimitDegrees=45
  swing2Motion=Locked
  twistLimitDegrees=45
  twistMotion=Locked
 }
 {
  bone1Name=dyn_Skirt_b_l_01
  bone2Name=base_Skirt_b_l_01
  swing1LimitDegrees=45
  swing1Motion=Free
  swing2LimitDegrees=45
  swing2Motion=Free
  twistLimitDegrees=45
  twistMotion=Free
 }
 {
  bone1Name=dyn_Skirt_b_l_02
  bone2Name=dyn_Skirt_b_l_01
  swing1LimitDegrees=45
  swing1Motion=Free
  swing2LimitDegrees=45
  swing2Motion=Free
  twistLimitDegrees=45
  twistMotion=Free
 }
 {
  bone1Name=dyn_Skirt_b_l_03
  bone2Name=dyn_Skirt_b_l_02
  swing1LimitDegrees=45
  swing1Motion=Free
  swing2LimitDegrees=45
  swing2Motion=Free
  twistLimitDegrees=45
  twistMotion=Free
 }
 {
  bone1Name=base_Skirt_b_r_01
  bone2Name=thigh_r
  swing1LimitDegrees=45
  swing1Motion=Locked
  swing2LimitDegrees=45
  swing2Motion=Locked
  twistLimitDegrees=45
  twistMotion=Locked
 }
 {
  bone1Name=dyn_Skirt_b_r_01
  bone2Name=base_Skirt_b_r_01
  swing1LimitDegrees=45
  swing1Motion=Free
  swing2LimitDegrees=45
  swing2Motion=Free
  twistLimitDegrees=45
  twistMotion=Free
 }
 {
  bone1Name=dyn_Skirt_b_r_02
  bone2Name=dyn_Skirt_b_r_01
  swing1LimitDegrees=45
  swing1Motion=Free
  swing2LimitDegrees=45
  swing2Motion=Free
  twistLimitDegrees=45
  twistMotion=Free
 }
 {
  bone1Name=dyn_Skirt_b_r_03
  bone2Name=dyn_Skirt_b_r_02
  swing1LimitDegrees=45
  swing1Motion=Free
  swing2LimitDegrees=45
  swing2Motion=Free
  twistLimitDegrees=45
  twistMotion=Free
 }
 {
  bone1Name=dyn_Coat_r_02
  bone2Name=dyn_Coat_f_r_03
  swing1LimitDegrees=45
  swing1Motion=Locked
  swing2LimitDegrees=45
  swing2Motion=Locked
  twistLimitDegrees=45
  twistMotion=Locked
 }
 {
  bone1Name=dyn_Coat_r_02
  bone2Name=dyn_Coat_b_r_02
  swing1LimitDegrees=45
  swing1Motion=Locked
  swing2LimitDegrees=45
  swing2Motion=Locked
  twistLimitDegrees=45
  twistMotion=Locked
 }
 {
  bone1Name=dyn_Coat_l_02
  bone2Name=dyn_Coat_b_l_02
  swing1LimitDegrees=45
  swing1Motion=Locked
  swing2LimitDegrees=45
  swing2Motion=Locked
  twistLimitDegrees=45
  twistMotion=Locked
 }
 {
  bone1Name=dyn_Coat_l_02
  bone2Name=dyn_Coat_f_l_03
  swing1LimitDegrees=45
  swing1Motion=Locked
  swing2LimitDegrees=45
  swing2Motion=Locked
  twistLimitDegrees=45
  twistMotion=Locked
 }
 {
  bone1Name=dyn_Coat_bA_r_03
  bone2Name=dyn_Coat_bA_l_03
  swing1LimitDegrees=45
  swing1Motion=Locked
  swing2LimitDegrees=45
  swing2Motion=Locked
  twistLimitDegrees=45
  twistMotion=Locked
 }
]
