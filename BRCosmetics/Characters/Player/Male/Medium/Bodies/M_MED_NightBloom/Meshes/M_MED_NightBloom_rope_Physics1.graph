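# Physics asset for the left and right dangling ropes on the NightBloom medium male player body cosmetic.
# Each rope is a chain of four thin capsules hung from a zero-size base body that is angularly locked to spine_05.
# Swing limits widen progressively from root to tip, while twist stays tight.
# Larger spine and clavicle capsules form upper-torso volumes, joined by constraints with free angular motion.
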PhysicsAsset /BRCosmetics/Characters/Player/Male/Medium/Bodies/M_MED_NightBloom/Meshes/M_MED_NightBloom_rope_Physics1
bodies {
 base_rope_l_01 [{shapeType=Capsule, length=0, radius=0, shapeName=base_rope_l_01_capsule}]
 base_rope_r_01 [{shapeType=Capsule, length=0, radius=0, shapeName=base_rope_r_01_capsule}]
 clavicle_l [{shapeType=Capsule, length=5.8415, radius=7.9585, shapeName=clavicle_l_capsule, center=<6.3622 1.3485 -4.2734>, rotation=<rot -90 -3.4325 0>}]
 clavicle_r [{shapeType=Capsule, length=6.6415, radius=7.1585, shapeName=clavicle_r_capsule, center=<-6.3221 -2.1454 4.2761>, rotation=<rot 90 -3.4325 0>}]
 dyn_rope_l_01 [{shapeType=Capsule, length=3.5, radius=0.7, shapeName=dyn_rope_l_01_capsule, center=<2.0461 0 0>, rotation=<rot 90 0 0>}]
 dyn_rope_l_02 [{shapeType=Capsule, length=3.4, radius=0.8, shapeName=dyn_rope_l_02_capsule, center=<2.0461 0 0.1494>, rotation=<rot 90 0 0>}]
 dyn_rope_l_03 [{shapeType=Capsule, length=3.4, radius=0.8, shapeName=dyn_rope_l_03_capsule, center=<2.0461 0 0.0752>, rotation=<rot 90 0 0>}]
 dyn_rope_l_04 [{shapeType=Capsule, length=3.4, radius=0.8, shapeName=dyn_rope_l_04_capsule, center=<2.0461 0 0.2069>, rotation=<rot 90 0 0>}]
 dyn_rope_r_01 [{shapeType=Capsule, length=3.5, radius=0.7, shapeName=dyn_rope_r_01_capsule, center=<2.0461 0 0>, rotation=<rot 90 0 0>}]
 dyn_rope_r_02 [{shapeType=Capsule, length=3.4, radius=0.8, shapeName=dyn_rope_r_02_capsule, center=<2.0461 0 0.1494>, rotation=<rot 90 0 0>}]
 dyn_rope_r_03 [{shapeType=Capsule, length=3.4, radius=0.8, shapeName=dyn_rope_r_03_capsule, center=<2.0461 0 0.1966>, rotation=<rot 90 0 0>}]
 dyn_rope_r_04 [{shapeType=Capsule, length=3.4, radius=0.8, shapeName=dyn_rope_r_04_capsule, center=<2.0461 0 0.215>, rotation=<rot 90 0 0>}]
 spine_04 [{shapeType=Capsule, length=10.5, radius=12.7, shapeName=spine_04_capsule, center=<2.5908 0.933 0>}]
 spine_05 [{shapeType=Capsule, length=17.7314, radius=11.2686, shapeName=spine_05_capsule, center=<-3.8103 1.5114 0>}]
}
constraints [
 {
  bone1Name=base_rope_l_01
  bone2Name=spine_05
  swing1LimitDegrees=45
  swing1Motion=Locked
  swing2LimitDegrees=45
  swing2Motion=Locked
  twistLimitDegrees=45
  twistMotion=Locked
 }
 {bone1Name=dyn_rope_l_01, bone2Name=base_rope_l_01, swing1LimitDegrees=10, swing2LimitDegrees=10, twistLimitDegrees=2}
 {bone1Name=dyn_rope_l_02, bone2Name=dyn_rope_l_01, swing1LimitDegrees=15, swing2LimitDegrees=15, twistLimitDegrees=2}
 {bone1Name=dyn_rope_l_03, bone2Name=dyn_rope_l_02, swing1LimitDegrees=20, swing2LimitDegrees=20, twistLimitDegrees=2}
 {bone1Name=dyn_rope_l_04, bone2Name=dyn_rope_l_03, swing1LimitDegrees=25, swing2LimitDegrees=25, twistLimitDegrees=2}
 {bone1Name=dyn_rope_r_01, bone2Name=base_rope_r_01, swing1LimitDegrees=10, swing2LimitDegrees=10, twistLimitDegrees=2}
 {
  bone1Name=base_rope_r_01
  bone2Name=spine_05
  swing1LimitDegrees=45
  swing1Motion=Locked
  swing2LimitDegrees=45
  swing2Motion=Locked
  twistLimitDegrees=45
  twistMotion=Locked
 }
 {bone1Name=dyn_rope_r_02, bone2Name=dyn_rope_r_01, swing1LimitDegrees=15, swing2LimitDegrees=15, twistLimitDegrees=2}
 {bone1Name=dyn_rope_r_03, bone2Name=dyn_rope_r_02, swing1LimitDegrees=20, swing2LimitDegrees=20, twistLimitDegrees=2}
 {bone1Name=dyn_rope_r_04, bone2Name=dyn_rope_r_03, swing1LimitDegrees=25, swing2LimitDegrees=25, twistLimitDegrees=2}
 {
  bone1Name=spine_05
  bone2Name=spine_04
  swing1LimitDegrees=45
  swing1Motion=Free
  swing2LimitDegrees=45
  swing2Motion=Free
  twistLimitDegrees=45
  twistMotion=Free
 }
 {
  bone1Name=clavicle_l
  bone2Name=spine_05
  swing1LimitDegrees=45
  swing1Motion=Free
  swing2LimitDegrees=45
  swing2Motion=Free
  twistLimitDegrees=45
  twistMotion=Free
 }
 {
  bone1Name=clavicle_r
  bone2Name=spine_05
  swing1LimitDegrees=45
  swing1Motion=Free
  swing2LimitDegrees=45
  swing2Motion=Free
  twistLimitDegrees=45
  twistMotion=Free
 }
]
