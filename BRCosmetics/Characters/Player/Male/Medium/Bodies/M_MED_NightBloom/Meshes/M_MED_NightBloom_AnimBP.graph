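# Secondary-motion AnimBlueprint for the NightBloom medium male player body, following the attached parent mesh's pose.
# The copied pose feeds a Control Rig driven by crouch and slide flags, then chained per-bone AnimDynamics springs for skirt, sleeves and collar.
# Two RigidBody nodes with separate physics assets, one for rope, finish the pose.
# The update event sets crouch-moving and a falling-dependent GravityOverride, which no AnimGraph node visibly reads.
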AnimBlueprint /BRCosmetics/Characters/Player/Male/Medium/Bodies/M_MED_NightBloom/Meshes/M_MED_NightBloom_AnimBP
target_skeleton=&/M_MED_NightBloom_Skeleton
graphs [
 AnimGraph {nodes [
   n1 = Control Rig!(
    Source=Copy Pose From Mesh() {bUseAttachedParent=true}
    Alpha
    "is Crouching ":bool=Get bIsCrouching()
    is sliding:bool=Get bIsKneeSliding()
    "is Crouch Moving ":bool="Get Crouch moving "() -> ("Crouch moving ":bool)
    "is Moving ":bool=false
   ) -> (Pose) {controlRigAssetReference {blueprintRigClass=&/M_MED_NightBloom_CtrlRig1.M_MED_NightBloom_CtrlRig1_C}}
   n2 = Anim Dynamics - Bone: dyn_Skirt_f_r_01(ComponentPose=Local To Component(LocalPose=@n1), Alpha=1, ExternalForce="0,0,0") {
    angularDampingOverride=0.9999
    angularSpringConstant=1000
    bAngularSpring=true
    bOverrideAngularDamping=true
    bUseGravityOverride=true
    gravityScale=0.35
    simulationSpace=RootRelative
    boundBone {boneName=dyn_Skirt_f_r_01}
    componentAppliedLinearAccClamp=<1000 1000 1000>
    componentLinearAccScale=<0.3 0.3 0.3>
    componentLinearVelScale=<0.3 0.3 0.3>
    gravityOverride {z=-980}
    physicsBodyDefinitions [{
      collisionType=CustomSphere
      sphereCollisionRadius=2
      boundBone {boneName=dyn_Skirt_f_r_01}
      boxExtents=<15 15 15>
      constraintSetup {angularLimitsMax=<2 0 5>, angularLimitsMin=<-2 -5 0>, angularTarget=<1 0 0>}
      localJointOffset=<3 -0.2 0>
     }]
    simSpaceSettings {maxAngularAcceleration=10, maxAngularVelocity=2, simSpaceAngularAlpha=1}
   }
   n3 = Anim Dynamics - Bone: dyn_Skirt_f_r_02(ComponentPose=@n2, Alpha=1, ExternalForce="0,0,0") {
    angularDampingOverride=0.9999
    angularSpringConstant=1000
    bAngularSpring=true
    bOverrideAngularDamping=true
    bUseGravityOverride=true
    gravityScale=0.35
    simulationSpace=RootRelative
    boundBone {boneName=dyn_Skirt_f_r_02}
    componentAppliedLinearAccClamp=<1000 1000 1000>
    componentLinearAccScale=<0.3 0.3 0.3>
    componentLinearVelScale=<0.3 0.3 0.3>
    gravityOverride {z=-980}
    physicsBodyDefinitions [{
      collisionType=CustomSphere
      sphereCollisionRadius=2
      boundBone {boneName=dyn_Skirt_f_r_02}
      boxExtents=<15 15 15>
      constraintSetup {angularLimitsMax=<2 0 5>, angularLimitsMin=<-2 -5 0>, angularTarget=<1 0 0>}
      localJointOffset=<3 -0.2 0>
     }]
    simSpaceSettings {maxAngularAcceleration=10, maxAngularVelocity=2, simSpaceAngularAlpha=1}
   }
   n4 = Anim Dynamics - Bone: dyn_Skirt_f_r_03(ComponentPose=@n3, Alpha=1, ExternalForce="0,0,0") {
    angularDampingOverride=0.9999
    angularSpringConstant=1000
    bAngularSpring=true
    bOverrideAngularDamping=true
    bUseGravityOverride=true
    gravityScale=0.35
    simulationSpace=RootRelative
    boundBone {boneName=dyn_Skirt_f_r_03}
    componentAppliedLinearAccClamp=<1000 1000 1000>
    componentLinearAccScale=<0.3 0.3 0.3>
    componentLinearVelScale=<0.3 0.3 0.3>
    gravityOverride {z=-980}
    physicsBodyDefinitions [{
      collisionType=CustomSphere
      sphereCollisionRadius=2
      boundBone {boneName=dyn_Skirt_f_r_03}
      boxExtents=<15 15 15>
      constraintSetup {angularLimitsMax=<2 0 5>, angularLimitsMin=<-2 -5 0>, angularTarget=<1 0 0>}
      localJointOffset=<3 -0.2 0>
     }]
    simSpaceSettings {maxAngularAcceleration=10, maxAngularVelocity=2, simSpaceAngularAlpha=1}
   }
   n5 = Anim Dynamics - Bone: dyn_Skirt_f_r_04(ComponentPose=@n4, Alpha=1, ExternalForce="0,0,0") {
    angularDampingOverride=0.9999
    angularSpringConstant=1000
    bAngularSpring=true
    bOverrideAngularDamping=true
    bUseGravityOverride=true
    gravityScale=0.35
    simulationSpace=RootRelative
    boundBone {boneName=dyn_Skirt_f_r_04}
    componentAppliedLinearAccClamp=<1000 1000 1000>
    componentLinearAccScale=<0.3 0.3 0.3>
    componentLinearVelScale=<0.3 0.3 0.3>
    gravityOverride {z=-980}
    physicsBodyDefinitions [{
      collisionType=CustomSphere
      sphereCollisionRadius=2
      boundBone {boneName=dyn_Skirt_f_r_04}
      boxExtents=<15 15 15>
      constraintSetup {angularLimitsMax=<2 2 5>, angularLimitsMin=<-2 -5 0>, angularTarget=<1 0 0>}
      localJointOffset=<3 -0.2 0>
     }]
    simSpaceSettings {maxAngularAcceleration=10, maxAngularVelocity=2, simSpaceAngularAlpha=1}
   }
   n6 = Anim Dynamics - Bone: dyn_Skirt_r_01(ComponentPose=@n5, Alpha=1, ExternalForce="0,0,0") {
    angularDampingOverride=0.9999
    angularSpringConstant=1000
    bAngularSpring=true
    bOverrideAngularDamping=true
    bUseGravityOverride=true
    gravityScale=0.35
    simulationSpace=RootRelative
    boundBone {boneName=dyn_Skirt_r_01}
    componentAppliedLinearAccClamp=<1000 1000 1000>
    componentLinearAccScale=<0.3 0.3 0.3>
    componentLinearVelScale=<0.3 0.3 0.3>
    gravityOverride {z=-980}
    physicsBodyDefinitions [{
      collisionType=CustomSphere
      sphereCollisionRadius=2
      boundBone {boneName=dyn_Skirt_r_01}
      boxExtents=<15 15 15>
      constraintSetup {angularLimitsMax=<2 0 0>, angularLimitsMin=<-2 -5 -1>, angularTarget=<1 0 0>}
      localJointOffset=<3 0 0>
     }]
    simSpaceSettings {maxAngularAcceleration=10, maxAngularVelocity=2, simSpaceAngularAlpha=1}
   }
   n7 = Anim Dynamics - Bone: dyn_Skirt_r_02(ComponentPose=@n6, Alpha=1, ExternalForce="0,0,0") {
    angularDampingOverride=0.9999
    angularSpringConstant=1000
    bAngularSpring=true
    bOverrideAngularDamping=true
    bUseGravityOverride=true
    gravityScale=0.35
    simulationSpace=RootRelative
    boundBone {boneName=dyn_Skirt_r_02}
    componentAppliedLinearAccClamp=<1000 1000 1000>
    componentLinearAccScale=<0.3 0.3 0.3>
    componentLinearVelScale=<0.3 0.3 0.3>
    gravityOverride {z=-980}
    physicsBodyDefinitions [{
      collisionType=CustomSphere
      sphereCollisionRadius=2
      boundBone {boneName=dyn_Skirt_r_02}
      boxExtents=<15 15 15>
      constraintSetup {angularLimitsMax=<2 0 0>, angularLimitsMin=<-2 -5 -1>, angularTarget=<1 0 0>}
      localJointOffset=<3 0 0>
     }]
    simSpaceSettings {maxAngularAcceleration=10, maxAngularVelocity=2, simSpaceAngularAlpha=1}
   }
   n8 = Anim Dynamics - Bone: dyn_Skirt_r_03(ComponentPose=@n7, Alpha=1, ExternalForce="0,0,0") {
    angularDampingOverride=0.9999
    angularSpringConstant=1000
    bAngularSpring=true
    bOverrideAngularDamping=true
    bUseGravityOverride=true
    gravityScale=0.35
    simulationSpace=RootRelative
    boundBone {boneName=dyn_Skirt_r_03}
    componentAppliedLinearAccClamp=<1000 1000 1000>
    componentLinearAccScale=<0.3 0.3 0.3>
    componentLinearVelScale=<0.3 0.3 0.3>
    gravityOverride {z=-980}
    physicsBodyDefinitions [{
      collisionType=CustomSphere
      sphereCollisionRadius=2
      boundBone {boneName=dyn_Skirt_r_03}
      boxExtents=<15 15 15>
      constraintSetup {angularLimitsMax=<2 0 0>, angularLimitsMin=<-2 -5 -1>, angularTarget=<1 0 0>}
      localJointOffset=<3 0 0>
     }]
    simSpaceSettings {maxAngularAcceleration=10, maxAngularVelocity=2, simSpaceAngularAlpha=1}
   }
   n9 = Anim Dynamics - Bone: dyn_Skirt_b_r_01(ComponentPose=@n8, Alpha=1, ExternalForce="0,0,0") {
    angularDampingOverride=0.9999
    angularSpringConstant=1000
    bAngularSpring=true
    bOverrideAngularDamping=true
    bUseGravityOverride=true
    gravityScale=0.35
    simulationSpace=RootRelative
    boundBone {boneName=dyn_Skirt_b_r_01}
    componentAppliedLinearAccClamp=<1000 1000 1000>
    componentLinearAccScale=<0.3 0.3 0.3>
    componentLinearVelScale=<0.3 0.3 0.3>
    gravityOverride {z=-980}
    physicsBodyDefinitions [{
      collisionType=CustomSphere
      sphereCollisionRadius=2
      boundBone {boneName=dyn_Skirt_b_r_01}
      boxExtents=<15 15 15>
      constraintSetup {angularLimitsMax=<2 0 0>, angularLimitsMin=<-2 -3 -5>, angularTarget=<1 0 0>}
      localJointOffset=<3 0 0>
     }]
    simSpaceSettings {maxAngularAcceleration=10, maxAngularVelocity=2, simSpaceAngularAlpha=1}
   }
   n10 = Anim Dynamics - Bone: dyn_Skirt_b_r_02(ComponentPose=@n9, Alpha=1, ExternalForce="0,0,0") {
    angularDampingOverride=0.9999
    angularSpringConstant=1000
    bAngularSpring=true
    bOverrideAngularDamping=true
    bUseGravityOverride=true
    gravityScale=0.35
    simulationSpace=RootRelative
    boundBone {boneName=dyn_Skirt_b_r_02}
    componentAppliedLinearAccClamp=<1000 1000 1000>
    componentLinearAccScale=<0.3 0.3 0.3>
    componentLinearVelScale=<0.3 0.3 0.3>
    gravityOverride {z=-980}
    physicsBodyDefinitions [{
      collisionType=CustomSphere
      sphereCollisionRadius=2
      boundBone {boneName=dyn_Skirt_b_r_02}
      boxExtents=<15 15 15>
      constraintSetup {angularLimitsMax=<2 0 0>, angularLimitsMin=<-2 -3 -5>, angularTarget=<1 0 0>}
      localJointOffset=<3 0 0>
     }]
    simSpaceSettings {maxAngularAcceleration=10, maxAngularVelocity=2, simSpaceAngularAlpha=1}
   }
   n11 = Anim Dynamics - Bone: dyn_Skirt_b_r_03(ComponentPose=@n10, Alpha=1, ExternalForce="0,0,0") {
    angularDampingOverride=0.9999
    angularSpringConstant=1000
    bAngularSpring=true
    bOverrideAngularDamping=true
    bUseGravityOverride=true
    gravityScale=0.35
    simulationSpace=RootRelative
    boundBone {boneName=dyn_Skirt_b_r_03}
    componentAppliedLinearAccClamp=<1000 1000 1000>
    componentLinearAccScale=<0.3 0.3 0.3>
    componentLinearVelScale=<0.3 0.3 0.3>
    gravityOverride {z=-980}
    physicsBodyDefinitions [{
      collisionType=CustomSphere
      sphereCollisionRadius=2
      boundBone {boneName=dyn_Skirt_b_r_03}
      boxExtents=<15 15 15>
      constraintSetup {angularLimitsMax=<2 0 0>, angularLimitsMin=<-2 -3 -5>, angularTarget=<1 0 0>}
      localJointOffset=<3 0 0>
     }]
    simSpaceSettings {maxAngularAcceleration=10, maxAngularVelocity=2, simSpaceAngularAlpha=1}
   }
   n12 = Anim Dynamics - Bone: dyn_Skirt_f_l_01(ComponentPose=@n11, Alpha=1, ExternalForce="0,0,0") {
    angularDampingOverride=0.9999
    angularSpringConstant=1000
    bAngularSpring=true
    bOverrideAngularDamping=true
    bUseGravityOverride=true
    gravityScale=0.35
    simulationSpace=RootRelative
    boundBone {boneName=dyn_Skirt_f_l_01}
    componentAppliedLinearAccClamp=<1000 1000 1000>
    componentLinearAccScale=<0.3 0.3 0.3>
    componentLinearVelScale=<0.3 0.3 0.3>
    gravityOverride {z=-980}
    physicsBodyDefinitions [{
      collisionType=CustomSphere
      sphereCollisionRadius=2
      boundBone {boneName=dyn_Skirt_f_l_01}
      boxExtents=<15 15 15>
      constraintSetup {angularLimitsMax=<2 0 0>, angularLimitsMin=<-2 -5 -5>, angularTarget=<1 0 0>}
      localJointOffset=<3 0.2 0>
     }]
    simSpaceSettings {maxAngularAcceleration=2, maxAngularVelocity=10, simSpaceAngularAlpha=1}
   }
   n13 = Anim Dynamics - Bone: dyn_Skirt_f_l_02(ComponentPose=@n12, Alpha=1, ExternalForce="0,0,0") {
    angularDampingOverride=0.9999
    angularSpringConstant=1000
    bAngularSpring=true
    bOverrideAngularDamping=true
    bUseGravityOverride=true
    gravityScale=0.35
    simulationSpace=RootRelative
    boundBone {boneName=dyn_Skirt_f_l_02}
    componentAppliedLinearAccClamp=<1000 1000 1000>
    componentLinearAccScale=<0.3 0.3 0.3>
    componentLinearVelScale=<0.3 0.3 0.3>
    gravityOverride {z=-980}
    physicsBodyDefinitions [{
      collisionType=CustomSphere
      sphereCollisionRadius=2
      boundBone {boneName=dyn_Skirt_f_l_02}
      boxExtents=<15 15 15>
      constraintSetup {angularLimitsMax=<2 0 0>, angularLimitsMin=<-2 -5 -5>, angularTarget=<1 0 0>}
      localJointOffset=<3 0.2 0>
     }]
    simSpaceSettings {maxAngularAcceleration=2, maxAngularVelocity=10, simSpaceAngularAlpha=1}
   }
   n14 = Anim Dynamics - Bone: dyn_Skirt_f_l_03(ComponentPose=@n13, Alpha=1, ExternalForce="0,0,0") {
    angularDampingOverride=0.9999
    angularSpringConstant=1000
    bAngularSpring=true
    bOverrideAngularDamping=true
    bUseGravityOverride=true
    gravityScale=0.35
    simulationSpace=RootRelative
    boundBone {boneName=dyn_Skirt_f_l_03}
    componentAppliedLinearAccClamp=<1000 1000 1000>
    componentLinearAccScale=<0.3 0.3 0.3>
    componentLinearVelScale=<0.3 0.3 0.3>
    gravityOverride {z=-980}
    physicsBodyDefinitions [{
      collisionType=CustomSphere
      sphereCollisionRadius=2
      boundBone {boneName=dyn_Skirt_f_l_03}
      boxExtents=<15 15 15>
      constraintSetup {angularLimitsMax=<2 0 0>, angularLimitsMin=<-2 -5 -5>, angularTarget=<1 0 0>}
      localJointOffset=<3 0.2 0>
     }]
    simSpaceSettings {maxAngularAcceleration=2, maxAngularVelocity=10, simSpaceAngularAlpha=1}
   }
   n15 = Anim Dynamics - Bone: dyn_Skirt_f_l_04(ComponentPose=@n14, Alpha=1, ExternalForce="0,0,0") {
    angularDampingOverride=0.9999
    angularSpringConstant=1000
    bAngularSpring=true
    bOverrideAngularDamping=true
    bUseGravityOverride=true
    gravityScale=0.35
    simulationSpace=RootRelative
    boundBone {boneName=dyn_Skirt_f_l_04}
    componentAppliedLinearAccClamp=<1000 1000 1000>
    componentLinearAccScale=<0.3 0.3 0.3>
    componentLinearVelScale=<0.3 0.3 0.3>
    gravityOverride {z=-980}
    physicsBodyDefinitions [{
      collisionType=CustomSphere
      sphereCollisionRadius=2
      boundBone {boneName=dyn_Skirt_f_l_04}
      boxExtents=<15 15 15>
      constraintSetup {angularLimitsMax=<2 2 0>, angularLimitsMin=<-2 -5 -5>, angularTarget=<1 0 0>}
      localJointOffset=<3 0.2 0>
     }]
    simSpaceSettings {maxAngularAcceleration=2, maxAngularVelocity=10, simSpaceAngularAlpha=1}
   }
   n16 = Anim Dynamics - Bone: dyn_Skirt_l_01(ComponentPose=@n15, Alpha=1, ExternalForce="0,0,0") {
    angularDampingOverride=0.9999
    angularSpringConstant=1000
    bAngularSpring=true
    bOverrideAngularDamping=true
    bUseGravityOverride=true
    gravityScale=0.35
    simulationSpace=RootRelative
    boundBone {boneName=dyn_Skirt_l_01}
    componentAppliedLinearAccClamp=<1000 1000 1000>
    componentLinearAccScale=<0.3 0.3 0.3>
    componentLinearVelScale=<0.3 0.3 0.3>
    gravityOverride {z=-980}
    physicsBodyDefinitions [{
      collisionType=CustomSphere
      sphereCollisionRadius=2
      boundBone {boneName=dyn_Skirt_l_01}
      boxExtents=<15 15 15>
      constraintSetup {angularLimitsMax=<2 0 1>, angularLimitsMin=<-2 -5 0>, angularTarget=<1 0 0>}
      localJointOffset=<3 0 0>
     }]
    simSpaceSettings {maxAngularAcceleration=2, maxAngularVelocity=10, simSpaceAngularAlpha=1}
   }
   n17 = Anim Dynamics - Bone: dyn_Skirt_l_02(ComponentPose=@n16, Alpha=1, ExternalForce="0,0,0") {
    angularDampingOverride=0.9999
    angularSpringConstant=1000
    bAngularSpring=true
    bOverrideAngularDamping=true
    bUseGravityOverride=true
    gravityScale=0.35
    simulationSpace=RootRelative
    boundBone {boneName=dyn_Skirt_l_02}
    componentAppliedLinearAccClamp=<1000 1000 1000>
    componentLinearAccScale=<0.3 0.3 0.3>
    componentLinearVelScale=<0.3 0.3 0.3>
    gravityOverride {z=-980}
    physicsBodyDefinitions [{
      collisionType=CustomSphere
      sphereCollisionRadius=2
      boundBone {boneName=dyn_Skirt_l_02}
      boxExtents=<15 15 15>
      constraintSetup {angularLimitsMax=<2 0 1>, angularLimitsMin=<-2 -5 0>, angularTarget=<1 0 0>}
      localJointOffset=<3 0 0>
     }]
    simSpaceSettings {maxAngularAcceleration=2, maxAngularVelocity=10, simSpaceAngularAlpha=1}
   }
   n18 = Anim Dynamics - Bone: dyn_Skirt_l_03(ComponentPose=@n17, Alpha=1, ExternalForce="0,0,0") {
    angularDampingOverride=0.9999
    angularSpringConstant=1000
    bAngularSpring=true
    bOverrideAngularDamping=true
    bUseGravityOverride=true
    gravityScale=0.35
    simulationSpace=RootRelative
    boundBone {boneName=dyn_Skirt_l_03}
    componentAppliedLinearAccClamp=<1000 1000 1000>
    componentLinearAccScale=<0.3 0.3 0.3>
    componentLinearVelScale=<0.3 0.3 0.3>
    gravityOverride {z=-980}
    physicsBodyDefinitions [{
      collisionType=CustomSphere
      sphereCollisionRadius=2
      boundBone {boneName=dyn_Skirt_l_03}
      boxExtents=<15 15 15>
      constraintSetup {angularLimitsMax=<2 0 5>, angularLimitsMin=<-2 -5 0>, angularTarget=<1 0 0>}
      localJointOffset=<3 0 0>
     }]
    simSpaceSettings {maxAngularAcceleration=2, maxAngularVelocity=10, simSpaceAngularAlpha=1}
   }
   n19 = Anim Dynamics - Bone: dyn_Skirt_b_l_01(ComponentPose=@n18, Alpha=1, ExternalForce="0,0,0") {
    angularDampingOverride=0.9999
    angularSpringConstant=1000
    bAngularSpring=true
    bOverrideAngularDamping=true
    bUseGravityOverride=true
    gravityScale=0.35
    simulationSpace=RootRelative
    boundBone {boneName=dyn_Skirt_b_l_01}
    componentAppliedLinearAccClamp=<1000 1000 1000>
    componentLinearAccScale=<0.3 0.3 0.3>
    componentLinearVelScale=<0.3 0.3 0.3>
    gravityOverride {z=-980}
    physicsBodyDefinitions [{
      collisionType=CustomSphere
      sphereCollisionRadius=2
      boundBone {boneName=dyn_Skirt_b_l_01}
      boxExtents=<15 15 15>
      constraintSetup {angularLimitsMax=<2 0 5>, angularLimitsMin=<-2 -3 0>, angularTarget=<1 0 0>}
      localJointOffset=<3 0 0>
     }]
    simSpaceSettings {maxAngularAcceleration=2, maxAngularVelocity=10, simSpaceAngularAlpha=1}
   }
   n20 = Anim Dynamics - Bone: dyn_Skirt_b_l_02(ComponentPose=@n19, Alpha=1, ExternalForce="0,0,0") {
    angularDampingOverride=0.9999
    angularSpringConstant=1000
    bAngularSpring=true
    bOverrideAngularDamping=true
    bUseGravityOverride=true
    gravityScale=0.35
    simulationSpace=RootRelative
    boundBone {boneName=dyn_Skirt_b_l_02}
    componentAppliedLinearAccClamp=<1000 1000 1000>
    componentLinearAccScale=<0.3 0.3 0.3>
    componentLinearVelScale=<0.3 0.3 0.3>
    gravityOverride {z=-980}
    physicsBodyDefinitions [{
      collisionType=CustomSphere
      sphereCollisionRadius=2
      boundBone {boneName=dyn_Skirt_b_l_02}
      boxExtents=<15 15 15>
      constraintSetup {angularLimitsMax=<2 0 5>, angularLimitsMin=<-2 -3 0>, angularTarget=<1 0 0>}
      localJointOffset=<3 0 0>
     }]
    simSpaceSettings {maxAngularAcceleration=2, maxAngularVelocity=10, simSpaceAngularAlpha=1}
   }
   n21 = Anim Dynamics - Bone: dyn_Skirt_b_l_03(ComponentPose=@n20, Alpha=1, ExternalForce="0,0,0") {
    angularDampingOverride=0.9999
    angularSpringConstant=1000
    bAngularSpring=true
    bOverrideAngularDamping=true
    bUseGravityOverride=true
    gravityScale=0.35
    simulationSpace=RootRelative
    boundBone {boneName=dyn_Skirt_b_l_03}
    componentAppliedLinearAccClamp=<1000 1000 1000>
    componentLinearAccScale=<0.3 0.3 0.3>
    componentLinearVelScale=<0.3 0.3 0.3>
    gravityOverride {z=-980}
    physicsBodyDefinitions [{
      collisionType=CustomSphere
      sphereCollisionRadius=2
      boundBone {boneName=dyn_Skirt_b_l_03}
      boxExtents=<15 15 15>
      constraintSetup {angularLimitsMax=<2 0 5>, angularLimitsMin=<-2 -3 0>, angularTarget=<1 0 0>}
      localJointOffset=<3 0 0>
     }]
    simSpaceSettings {maxAngularAcceleration=2, maxAngularVelocity=10, simSpaceAngularAlpha=1}
   }
   n22 = Anim Dynamics - Bone: dyn_Skirt_b_01(ComponentPose=@n21, Alpha=1, ExternalForce="0,0,0") {
    angularDampingOverride=0.9999
    angularSpringConstant=1000
    bAngularSpring=true
    bOverrideAngularDamping=true
    bUseGravityOverride=true
    gravityScale=0.35
    simulationSpace=RootRelative
    boundBone {boneName=dyn_Skirt_b_01}
    componentAppliedLinearAccClamp=<1000 1000 1000>
    componentLinearAccScale=<0.3 0.3 0.3>
    componentLinearVelScale=<0.3 0.3 0.3>
    gravityOverride {z=-980}
    physicsBodyDefinitions [{
      collisionType=CustomSphere
      sphereCollisionRadius=2
      boundBone {boneName=dyn_Skirt_b_01}
      boxExtents=<15 15 15>
      constraintSetup {angularLimitsMax=<2 0 3>, angularLimitsMin=<-2 -5 -3>, angularTarget=<1 0 0>}
      localJointOffset=<3 0 0>
     }]
    simSpaceSettings {maxAngularAcceleration=2, maxAngularVelocity=10, simSpaceAngularAlpha=1}
   }
   n23 = Anim Dynamics - Bone: dyn_Skirt_b_02(ComponentPose=@n22, Alpha=1, ExternalForce="0,0,0") {
    angularDampingOverride=0.9999
    angularSpringConstant=1000
    bAngularSpring=true
    bOverrideAngularDamping=true
    bUseGravityOverride=true
    gravityScale=0.35
    simulationSpace=RootRelative
    boundBone {boneName=dyn_Skirt_b_02}
    componentAppliedLinearAccClamp=<1000 1000 1000>
    componentLinearAccScale=<0.3 0.3 0.3>
    componentLinearVelScale=<0.3 0.3 0.3>
    gravityOverride {z=-980}
    physicsBodyDefinitions [{
      collisionType=CustomSphere
      sphereCollisionRadius=2
      boundBone {boneName=dyn_Skirt_b_02}
      boxExtents=<15 15 15>
      constraintSetup {angularLimitsMax=<2 0 3>, angularLimitsMin=<-2 -5 -3>, angularTarget=<1 0 0>}
      localJointOffset=<3 0 0>
     }]
    simSpaceSettings {maxAngularAcceleration=2, maxAngularVelocity=10, simSpaceAngularAlpha=1}
   }
   n24 = Anim Dynamics - Bone: dyn_Skirt_b_03(ComponentPose=@n23, Alpha=1, ExternalForce="0,0,0") {
    angularDampingOverride=0.9999
    angularSpringConstant=1000
    bAngularSpring=true
    bOverrideAngularDamping=true
    bUseGravityOverride=true
    gravityScale=0.35
    simulationSpace=RootRelative
    boundBone {boneName=dyn_Skirt_b_03}
    componentAppliedLinearAccClamp=<1000 1000 1000>
    componentLinearAccScale=<0.3 0.3 0.3>
    componentLinearVelScale=<0.3 0.3 0.3>
    gravityOverride {z=-980}
    physicsBodyDefinitions [{
      collisionType=CustomSphere
      sphereCollisionRadius=2
      boundBone {boneName=dyn_Skirt_b_03}
      boxExtents=<15 15 15>
      constraintSetup {angularLimitsMax=<2 0 3>, angularLimitsMin=<-2 -5 -3>, angularTarget=<1 0 0>}
      localJointOffset=<3 0 0>
     }]
    simSpaceSettings {maxAngularAcceleration=2, maxAngularVelocity=10, simSpaceAngularAlpha=1}
   }
   n25 = Anim Dynamics - Bone: dyn_sleeveFront_r(ComponentPose=@n24, Alpha=1, ExternalForce="0,0,0") {
    angularDampingOverride=0.9999
    angularSpringConstant=1000
    bAngularSpring=true
    bOverrideAngularDamping=true
    bUseGravityOverride=true
    gravityScale=0.35
    simulationSpace=BoneRelative
    boundBone {boneName=dyn_sleeveFront_r}
    componentAppliedLinearAccClamp=<1000 1000 1000>
    componentLinearAccScale=<0.3 0.3 0.3>
    componentLinearVelScale=<0.3 0.3 0.3>
    gravityOverride {z=-300}
    physicsBodyDefinitions [{
      collisionType=CustomSphere
      sphereCollisionRadius=2
      boundBone {boneName=dyn_sleeveFront_r}
      boxExtents=<15 15 15>
      constraintSetup {angularLimitsMax=<2 5 2>, angularLimitsMin=<-2 -5 -20>, angularTarget=<1 0 0>}
      localJointOffset=<-1.5 -0.5 0>
     }]
    relativeSpaceBone {boneName=upperarm_r}
    simSpaceSettings {maxAngularAcceleration=2, maxAngularVelocity=10, simSpaceAngularAlpha=1}
   }
   n26 = Anim Dynamics - Bone: dyn_sleeveBottom_r(ComponentPose=@n25, Alpha=1, ExternalForce="0,0,0") {
    angularDampingOverride=0.9999
    angularSpringConstant=1000
    bAngularSpring=true
    bOverrideAngularDamping=true
    bUseGravityOverride=true
    gravityScale=0.35
    simulationSpace=BoneRelative
    boundBone {boneName=dyn_sleeveBottom_r}
    componentAppliedLinearAccClamp=<1000 1000 1000>
    componentLinearAccScale=<0.3 0.3 0.3>
    componentLinearVelScale=<0.3 0.3 0.3>
    gravityOverride {z=-300}
    physicsBodyDefinitions [{
      collisionType=CustomSphere
      sphereCollisionRadius=2
      boundBone {boneName=dyn_sleeveBottom_r}
      boxExtents=<15 15 15>
      constraintSetup {angularLimitsMax=<2 2 5>, angularLimitsMin=<-2 -2 -20>, angularTarget=<1 0 0>}
      localJointOffset=<-1.5 -0.5 0>
     }]
    relativeSpaceBone {boneName=upperarm_r}
    simSpaceSettings {maxAngularAcceleration=2, maxAngularVelocity=10, simSpaceAngularAlpha=1}
   }
   n27 = Anim Dynamics - Bone: dyn_sleeveBack_r(ComponentPose=@n26, Alpha=1, ExternalForce="0,0,0") {
    angularDampingOverride=0.9999
    angularSpringConstant=1000
    bAngularSpring=true
    bOverrideAngularDamping=true
    bUseGravityOverride=true
    gravityScale=0.35
    simulationSpace=BoneRelative
    boundBone {boneName=dyn_sleeveBack_r}
    componentAppliedLinearAccClamp=<1000 1000 1000>
    componentLinearAccScale=<0.3 0.3 0.3>
    componentLinearVelScale=<0.3 0.3 0.3>
    gravityOverride {z=-300}
    physicsBodyDefinitions [{
      collisionType=CustomSphere
      sphereCollisionRadius=2
      boundBone {boneName=dyn_sleeveBack_r}
      boxExtents=<15 15 15>
      constraintSetup {angularLimitsMax=<2 2 5>, angularLimitsMin=<-2 -2 -20>, angularTarget=<1 0 0>}
      localJointOffset=<-1.5 -0.5 0>
     }]
    relativeSpaceBone {boneName=upperarm_r}
    simSpaceSettings {maxAngularAcceleration=2, maxAngularVelocity=10, simSpaceAngularAlpha=1}
   }
   n28 = Anim Dynamics - Bone: dyn_sleeveTop_r(ComponentPose=@n27, Alpha=1, ExternalForce="0,0,0") {
    angularDampingOverride=0.9999
    angularSpringConstant=1000
    bAngularSpring=true
    bOverrideAngularDamping=true
    bUseGravityOverride=true
    gravityScale=0.35
    simulationSpace=BoneRelative
    boundBone {boneName=dyn_sleeveTop_r}
    componentAppliedLinearAccClamp=<1000 1000 1000>
    componentLinearAccScale=<0.3 0.3 0.3>
    componentLinearVelScale=<0.3 0.3 0.3>
    gravityOverride {z=-300}
    physicsBodyDefinitions [{
      collisionType=CustomSphere
      sphereCollisionRadius=2
      boundBone {boneName=dyn_sleeveTop_r}
      boxExtents=<15 15 15>
      constraintSetup {angularLimitsMax=<2 0.5 5>, angularLimitsMin=<-2 -0.5 -20>, angularTarget=<1 0 0>}
      localJointOffset=<-1.5 -0.5 0>
     }]
    relativeSpaceBone {boneName=upperarm_r}
    simSpaceSettings {maxAngularAcceleration=2, maxAngularVelocity=10, simSpaceAngularAlpha=1}
   }
   n29 = Anim Dynamics - Bone: dyn_sleeveFront_l(ComponentPose=@n28, Alpha=1, ExternalForce="0,0,0") {
    angularDampingOverride=0.9999
    angularSpringConstant=1000
    bAngularSpring=true
    bOverrideAngularDamping=true
    bUseGravityOverride=true
    gravityScale=0.35
    simulationSpace=BoneRelative
    boundBone {boneName=dyn_sleeveFront_l}
    componentAppliedLinearAccClamp=<1000 1000 1000>
    componentLinearAccScale=<0.3 0.3 0.3>
    componentLinearVelScale=<0.3 0.3 0.3>
    gravityOverride {z=-300}
    physicsBodyDefinitions [{
      collisionType=CustomSphere
      sphereCollisionRadius=2
      boundBone {boneName=dyn_sleeveFront_l}
      boxExtents=<10 10 10>
      constraintSetup {angularLimitsMax=<2 2 5>, angularLimitsMin=<-2 -2 -20>, angularTarget=<1 0 0>}
      localJointOffset=<-1.5 0 0>
     }]
    relativeSpaceBone {boneName=upperarm_l}
    simSpaceSettings {maxAngularAcceleration=2, maxAngularVelocity=10, simSpaceAngularAlpha=1}
   }
   n30 = Anim Dynamics - Bone: dyn_sleeveBack_l(ComponentPose=@n29, Alpha=1, ExternalForce="0,0,0") {
    angularDampingOverride=0.9999
    angularSpringConstant=1000
    bAngularSpring=true
    bOverrideAngularDamping=true
    bUseGravityOverride=true
    gravityScale=0.35
    simulationSpace=BoneRelative
    boundBone {boneName=dyn_sleeveBack_l}
    componentAppliedLinearAccClamp=<1000 1000 1000>
    componentLinearAccScale=<0.3 0.3 0.3>
    componentLinearVelScale=<0.3 0.3 0.3>
    gravityOverride {z=-300}
    physicsBodyDefinitions [{
      collisionType=CustomSphere
      sphereCollisionRadius=2
      boundBone {boneName=dyn_sleeveBack_l}
      boxExtents=<10 10 10>
      constraintSetup {angularLimitsMax=<2 2 5>, angularLimitsMin=<-2 -2 -20>, angularTarget=<1 0 0>}
      localJointOffset=<-1.5 0 0>
     }]
    relativeSpaceBone {boneName=upperarm_l}
    simSpaceSettings {maxAngularAcceleration=2, maxAngularVelocity=10, simSpaceAngularAlpha=1}
   }
   n31 = Anim Dynamics - Bone: dyn_sleeveBottom_l(ComponentPose=@n30, Alpha=1, ExternalForce="0,0,0") {
    angularDampingOverride=0.9999
    angularSpringConstant=1000
    bAngularSpring=true
    bOverrideAngularDamping=true
    bUseGravityOverride=true
    gravityScale=0.35
    simulationSpace=BoneRelative
    boundBone {boneName=dyn_sleeveBottom_l}
    componentAppliedLinearAccClamp=<1000 1000 1000>
    componentLinearAccScale=<0.3 0.3 0.3>
    componentLinearVelScale=<0.3 0.3 0.3>
    gravityOverride {z=-300}
    physicsBodyDefinitions [{
      collisionType=CustomSphere
      sphereCollisionRadius=2
      boundBone {boneName=dyn_sleeveBottom_l}
      boxExtents=<10 10 10>
      constraintSetup {angularLimitsMax=<2 2 5>, angularLimitsMin=<-2 -2 -20>, angularTarget=<1 0 0>}
      localJointOffset=<-1.5 0 0>
     }]
    relativeSpaceBone {boneName=upperarm_l}
    simSpaceSettings {maxAngularAcceleration=2, maxAngularVelocity=10, simSpaceAngularAlpha=1}
   }
   n32 = Anim Dynamics - Bone: dyn_sleeveTop_l(ComponentPose=@n31, Alpha=1, ExternalForce="0,0,0") {
    angularDampingOverride=0.9999
    angularSpringConstant=1000
    bAngularSpring=true
    bOverrideAngularDamping=true
    bUseGravityOverride=true
    gravityScale=0.35
    simulationSpace=BoneRelative
    boundBone {boneName=dyn_sleeveTop_l}
    componentAppliedLinearAccClamp=<1000 1000 1000>
    componentLinearAccScale=<0.3 0.3 0.3>
    componentLinearVelScale=<0.3 0.3 0.3>
    gravityOverride {z=-300}
    physicsBodyDefinitions [{
      collisionType=CustomSphere
      sphereCollisionRadius=2
      boundBone {boneName=dyn_sleeveTop_l}
      boxExtents=<10 10 10>
      constraintSetup {angularLimitsMax=<2 2 5>, angularLimitsMin=<-2 -2 -20>, angularTarget=<1 0 0>}
      localJointOffset=<-1.5 0 0>
     }]
    relativeSpaceBone {boneName=upperarm_l}
    simSpaceSettings {maxAngularAcceleration=2, maxAngularVelocity=10, simSpaceAngularAlpha=1}
   }
   n33 = Anim Dynamics - Bone: dyn_collar_rf_01(ComponentPose=@n32, Alpha=1, ExternalForce="0,0,0") {
    angularDampingOverride=0.9999
    angularSpringConstant=1000
    bAngularSpring=true
    bOverrideAngularDamping=true
    bUseGravityOverride=true
    gravityScale=0.35
    simulationSpace=RootRelative
    boundBone {boneName=dyn_collar_rf_01}
    componentAppliedLinearAccClamp=<1000 1000 1000>
    componentLinearAccScale=<0.3 0.3 0.3>
    componentLinearVelScale=<0.3 0.3 0.3>
    gravityOverride {z=-300}
    physicsBodyDefinitions [{
      collisionType=CustomSphere
      sphereCollisionRadius=2
      boundBone {boneName=dyn_collar_rf_01}
      boxExtents=<15 15 15>
      constraintSetup {angularLimitsMax=<2 0 2>, angularLimitsMin=<-2 -5 -2>, angularTarget=<1 0 0>}
      localJointOffset=<-1.5 -0.5 0>
     }]
    relativeSpaceBone {boneName=spine_05}
    simSpaceSettings {maxAngularAcceleration=2, maxAngularVelocity=10, simSpaceAngularAlpha=1}
   }
   n34 = Anim Dynamics - Bone: dyn_collar_rf_02(ComponentPose=@n33, Alpha=1, ExternalForce="0,0,0") {
    angularDampingOverride=0.9999
    angularSpringConstant=1000
    bAngularSpring=true
    bOverrideAngularDamping=true
    bUseGravityOverride=true
    gravityScale=0.35
    simulationSpace=RootRelative
    boundBone {boneName=dyn_collar_rf_02}
    componentAppliedLinearAccClamp=<1000 1000 1000>
    componentLinearAccScale=<0.3 0.3 0.3>
    componentLinearVelScale=<0.3 0.3 0.3>
    gravityOverride {z=-300}
    physicsBodyDefinitions [{
      collisionType=CustomSphere
      sphereCollisionRadius=2
      boundBone {boneName=dyn_collar_rf_02}
      boxExtents=<15 15 15>
      constraintSetup {angularLimitsMax=<2 0 2>, angularLimitsMin=<-2 -5 -2>, angularTarget=<1 0 0>}
      localJointOffset=<-1.5 -0.5 0>
     }]
    relativeSpaceBone {boneName=spine_05}
    simSpaceSettings {maxAngularAcceleration=2, maxAngularVelocity=10, simSpaceAngularAlpha=1}
   }
   n35 = Anim Dynamics - Bone: dyn_collar_lf_01(ComponentPose=@n34, Alpha=1, ExternalForce="0,0,0") {
    angularDampingOverride=0.9999
    angularSpringConstant=1000
    bAngularSpring=true
    bOverrideAngularDamping=true
    bUseGravityOverride=true
    gravityScale=0.35
    simulationSpace=RootRelative
    boundBone {boneName=dyn_collar_lf_01}
    componentAppliedLinearAccClamp=<1000 1000 1000>
    componentLinearAccScale=<0.3 0.3 0.3>
    componentLinearVelScale=<0.3 0.3 0.3>
    gravityOverride {z=-300}
    physicsBodyDefinitions [{
      collisionType=CustomSphere
      sphereCollisionRadius=2
      boundBone {boneName=dyn_collar_lf_01}
      boxExtents=<15 15 15>
      constraintSetup {angularLimitsMax=<2 0 2>, angularLimitsMin=<-2 -5 -2>, angularTarget=<1 0 0>}
      localJointOffset=<-1.5 -0.5 0>
     }]
    relativeSpaceBone {boneName=spine_05}
    simSpaceSettings {maxAngularAcceleration=2, maxAngularVelocity=10, simSpaceAngularAlpha=1}
   }
   n36 = Anim Dynamics - Bone: dyn_collar_lf_02(ComponentPose=@n35, Alpha=1, ExternalForce="0,0,0") {
    angularDampingOverride=0.9999
    angularSpringConstant=1000
    bAngularSpring=true
    bOverrideAngularDamping=true
    bUseGravityOverride=true
    gravityScale=0.35
    simulationSpace=RootRelative
    boundBone {boneName=dyn_collar_lf_02}
    componentAppliedLinearAccClamp=<1000 1000 1000>
    componentLinearAccScale=<0.3 0.3 0.3>
    componentLinearVelScale=<0.3 0.3 0.3>
    gravityOverride {z=-300}
    physicsBodyDefinitions [{
      collisionType=CustomSphere
      sphereCollisionRadius=2
      boundBone {boneName=dyn_collar_lf_02}
      boxExtents=<15 15 15>
      constraintSetup {angularLimitsMax=<2 0 2>, angularLimitsMin=<-2 -5 -2>, angularTarget=<1 0 0>}
      localJointOffset=<-1.5 -0.5 0>
     }]
    relativeSpaceBone {boneName=spine_05}
    simSpaceSettings {maxAngularAcceleration=2, maxAngularVelocity=10, simSpaceAngularAlpha=1}
   }
   n37 = RigidBody!(ComponentPose=@n36, Alpha, OverrideWorldGravity:Vector="0,0,0", ExternalForce) -> (Pose) {
    bClampLinearTranslationLimitToRefPose=true
    bOverrideWorldGravity=true
    overridePhysicsAsset=&/M_MED_NightBloom_Physics1
    simulationSpace=BaseBoneSpace
    baseBoneRef {boneName=spine_02}
    overrideWorldGravity {z=-980}
    simSpaceSettings {
     dampingAlpha=0.5
     maxAngularAcceleration=100
     maxAngularVelocity=5
     maxLinearAcceleration=500
     maxLinearVelocity=100
     velocityScaleZ=0
     worldAlpha=1
    }
   }
   n38 = RigidBody(ComponentPose=@n37, OverrideWorldGravity="0,0,0") {
    bClampLinearTranslationLimitToRefPose=true
    bOverrideWorldGravity=true
    overridePhysicsAsset=&/M_MED_NightBloom_rope_Physics1
    simulationSpace=BaseBoneSpace
    baseBoneRef {boneName=spine_05}
    overrideWorldGravity {z=-980}
    simSpaceSettings {
     dampingAlpha=0.5
     maxAngularAcceleration=100
     maxAngularVelocity=5
     maxLinearAcceleration=500
     maxLinearVelocity=100
     velocityScaleZ=0
     worldAlpha=1
    }
   }
   Output Pose(Result=Component To Local(ComponentPose=@n38)) {layerGroup=None}
  ]}
 EventGraph {nodes [
   Event BlueprintUpdateAnimation()
   > n1 = Sequence()
   > "Set Crouch moving "("Crouch moving ":bool=AND Boolean(A=Get bIsCrouching(), B=Get bIsMoving2D())) -> (Output_Get:bool)
   n2 = "Gravity Override State Function (Parameter Asset) (Nativized)"(
    execute=@n1.then_1
    CharacterPartAnimInstance=Self-Reference()
    GameTimeInSeconds=GetGameTimeInSeconds()
    Parameters=&/F_MED_NightBloom_Parameters1
   )
   > n3 = Branch(Condition=Get bIsFalling())
   > Set GravityOverride(GravityOverride="0,0,980")
   Set GravityOverride(execute=@n3.else, GravityOverride=@n2)
  ]}
]
variables [
 "Crouch moving " {type=boolean}
 GravityOverride {type=Vector}
]
