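# Physics asset of cape colliders for the male medium NightBloom player body.
# Capsules on the pelvis, spine, clavicles and neck, plus a head sphere, approximate the upper body; arm, leg and foot bodies have no shapes.
# Six dynamic coat bones at the _03 joint carry their own capsules, each constrained to spine_02 with free swing and twist, so the listed 45-degree limits are inactive.
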PhysicsAsset /BRCosmetics/Characters/Player/Male/Medium/Bodies/M_MED_NightBloom/Meshes/M_MED_NightBloom_CapeColliders
bodies {
 calf_l []
 calf_r []
 clavicle_l [{shapeType=Capsule, length=8.1298, radius=11.157, center=<9.4266 -1.1806 -2.4977>, rotation=<rot 69.3297 1.4476 176.9462>}]
 clavicle_r [{shapeType=Capsule, length=7.7493, radius=11.157, center=<-9.4266 1.1806 2.4977>, rotation=<rot 69.3297 1.4476 -3.0538>}]
 dyn_Coat_bA_l_03 [{shapeType=Capsule, length=11.3151, radius=9.9217, shapeName=dyn_Coat_bA_l_03_capsule, center=<3.046 5.1251 -7.428>, rotation=<rot 0 0 90>}]
 dyn_Coat_bA_r_03 [{shapeType=Capsule, length=10.1619, radius=9.293, shapeName=dyn_Coat_bA_r_03_capsule, center=<3.046 -5.1251 -7.428>, rotation=<rot 0 0 90>}]
 dyn_Coat_b_l_03 [{
   shapeType=Capsule
   length=10.1049
   radius=8.6397
   shapeName=dyn_Coat_b_l_03_capsule
   center=<3.0648 2.2422 -5.3038>
   rotation=<rot -0.0002 -10.0328 66.7652>
  }]
 dyn_Coat_b_r_03 [{
   shapeType=Capsule
   length=10.016
   radius=7.8705
   shapeName=dyn_Coat_b_r_03_capsule
   center=<2.9192 -1.4033 -5.5931>
   rotation=<rot -0.0001 10.1454 104.299>
  }]
 dyn_Coat_l_03 [{
   shapeType=Capsule
   length=11.4401
   radius=9.3541
   shapeName=dyn_Coat_l_03_capsule
   center=<3.0432 -0.4095 -5.1983>
   rotation=<rot 0 -15.6274 90>
  }]
 dyn_Coat_r_03 [{
   shapeType=Capsule
   length=11.9101
   radius=9.3699
   shapeName=dyn_Coat_r_03_capsule
   center=<3.0853 0.2795 -5.1983>
   rotation=<rot 0 10.5134 90>
  }]
 foot_l []
 foot_r []
 head [{shapeType=Sphere, radius=17.3125, center=<11.028 -0.174 0>}]
 lowerarm_l []
 lowerarm_r []
 neck_02 [{shapeType=Capsule, length=13.4375, radius=7.3175, center=<-0.2642 -0.3408 0>, rotation=<rot 90 -171.1193 -175.3978>}]
 pelvis [{shapeType=Capsule, length=5.5341, radius=17.6503, center=<-2.4887 0.1881 0>}]
 spine_01 [{shapeType=Capsule, length=4.7498, radius=16.7526, center=<-1.7965 0.1884 0>}]
 spine_02 [
  {shapeType=Capsule, length=6.197, radius=14.7616, center=<-1.711 -0.1851 0.0001>}
  {shapeType=Capsule, length=29.3956, radius=11.5807, center=<2.5024 0.4992 3.9825>, rotation=<rot 90 0.0049 0.0049>}
  {shapeType=Capsule, length=29.875, radius=11.1013, center=<2.4704 0.4991 -3.9825>, rotation=<rot 90 -0.0396 -0.0396>}
 ]
 spine_03 [{shapeType=Capsule, length=7.5653, radius=14.3484, center=<0.1445 0.2396 0.0001>}]
 spine_04 [{shapeType=Capsule, length=8.785, radius=14.7879, center=<2.766 0.7765 0.0002>}]
 spine_05 [
  {shapeType=Capsule, length=7.6022, radius=16.1155, center=<-2.2795 1.8109 0>}
  {shapeType=Capsule, length=13.9518, radius=13.9425, center=<-0.2916 0.4452 0>}
 ]
 thigh_l []
 thigh_r []
 upperarm_l []
 upperarm_r []
}
constraints [
 {
  bone1Name=dyn_Coat_bA_l_03
  bone2Name=spine_02
  swing1LimitDegrees=45
  swing1Motion=Free
  swing2LimitDegrees=45
  swing2Motion=Free
  twistLimitDegrees=45
  twistMotion=Free
 }
 {
  bone1Name=dyn_Coat_bA_r_03
  bone2Name=spine_02
  swing1LimitDegrees=45
  swing1Motion=Free
  swing2LimitDegrees=45
  swing2Motion=Free
  twistLimitDegrees=45
  twistMotion=Free
 }
 {
  bone1Name=dyn_Coat_b_l_03
  bone2Name=spine_02
  swing1LimitDegrees=45
  swing1Motion=Free
  swing2LimitDegrees=45
  swing2Motion=Free
  twistLimitDegrees=45
  twistMotion=Free
 }
 {
  bone1Name=dyn_Coat_b_r_03
  bone2Name=spine_02
  swing1LimitDegrees=45
  swing1Motion=Free
  swing2LimitDegrees=45
  swing2Motion=Free
  twistLimitDegrees=45
  twistMotion=Free
 }
 {
  bone1Name=dyn_Coat_l_03
  bone2Name=spine_02
  swing1LimitDegrees=45
  swing1Motion=Free
  swing2LimitDegrees=45
  swing2Motion=Free
  twistLimitDegrees=45
  twistMotion=Free
 }
 {
  bone1Name=dyn_Coat_r_03
  bone2Name=spine_02
  swing1LimitDegrees=45
  swing1Motion=Free
  swing2LimitDegrees=45
  swing2Motion=Free
  twistLimitDegrees=45
  twistMotion=Free
 }
]
